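# TurnTableGray is a Default Lit material in the Turntable tool folder, named as a gray surface for that tool.
# Its graph holds a single scalar parameter, Value, with a dark default that material instances can override.
# The file shows no link from that parameter to a material output and no textures or other logic.
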Material /Game/Tools/Turntable/Material/TurnTableGray
graph {
 properties {pixelDepthOffsetMode=PDOM_Legacy, usedShadingModels=Default Lit}
 nodes [ScalarParameter() {defaultValue=0.0459, parameterName=Value}]
}
